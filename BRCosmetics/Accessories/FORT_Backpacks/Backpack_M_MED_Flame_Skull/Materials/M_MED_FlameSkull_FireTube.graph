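# Material for the Flame Skull backpack's fire tube, faking a glossy glass tube with animated flame inside on a skeletal mesh.
# Core color mixes backscatter and sunlight-tinted scattering, computed from the collection's sun vector, normal and camera direction.
# Bump-offset, noise-warped scrolling textures drive flame color, which multiplies the core into emissive and fades at the Fresnel rim.
# An FX texture mask blends this black-base layer over the shared character PreFX material before PostFX.
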Material /BRCosmetics/Accessories/FORT_Backpacks/Backpack_M_MED_Flame_Skull/Materials/M_MED_FlameSkull_FireTube
graph {
 properties {
  bUseMaterialAttributes=true
  bUsedWithMeshDeformer=true
  bUsedWithSkeletalMesh=true
  pixelDepthOffsetMode=PDOM_Legacy
  usedShadingModels=Default Lit
 }
 nodes [
  n1 = VectorParameter() {parameterName=Core Color, defaultValue=<rgba 0 0.0568 0.6 1>}
  n2 = VectorParameter() {parameterName=BackscatterColor, defaultValue=<rgba 0.0015 0.8 0.5572 1>}
  n3 = CollectionParameter() {
   collection=&FortniteMaterialParameters
   parameterName=SunlightDirectionalVector
  }
  n4 = ComponentMask(None=@n3) {b=true, g=true, r=true}
  n5 = PixelNormalWS()
  n6 = Multiply(A=Multiply(A=DotProduct(A=@n4, B=@n5)) {constB=2}, B=@n5)
  n7 = DotProduct(
   A=Normalize(VectorInput=Subtract(A=@n6, B=@n4))
   B=Multiply(A=MaterialFunctionCall() {materialFunction=&CameraDirectionVector}.Result) {constB=-1}
  )
  n8 = Max(A=Add(A=Multiply(A=@n7) {constB=-1}, B=ScalarParameter() {defaultValue=0.1, group=Slurp, parameterName=Slurp_InnerBackScatteringBias})) {constB=0}
  n9 = ScalarParameter() {defaultValue=1, group=Slurp, parameterName=Slurp_InnerBackScatteringPower}
  n10 = Saturate(None=Multiply(A=Power(Base=@n8, Exp=@n9), B=ScalarParameter() {defaultValue=0.1, group=Slurp, parameterName=Slurp_InnerBackScatteringScale}) {constB=0.55})
  n11 = CollectionParameter() {
   collection=&FortniteMaterialParameters
   parameterName=SunLightColorWithoutLightningFlashes
  }
  n12 = ComponentMask(None=Saturate(None=@n11)) {b=true, g=true, r=true}
  n13 = VectorParameter() {parameterName=Subsurface Color, defaultValue=<rgba 0.3672 1.1519 1.5 1>}
  n14 = Power(Base=Max(A=@n7) {constB=0}, Exp=ScalarParameter() {defaultValue=4.5, group=Slurp, parameterName=Slurp_InnerScatteringPower})
  n15 = ScalarParameter() {defaultValue=0.8, group=Slurp, parameterName=Slurp_InnerScatteringScale}
  n16 = Multiply(
   A=LinearInterpolate(A=LinearInterpolate(A=@n1.RGB, B=@n2.RGB, Alpha=@n10), B=Multiply(A=@n12, B=@n13.RGB), Alpha=Multiply(A=@n14, B=@n15) {constB=0.4})
   B=ScalarParameter() {defaultValue=0.5, parameterName=Core Brightness}
  )
  n17 = MaterialFunctionCall() {materialFunction=&BaseFN_CharacterFunction_PreFX}
  n18 = Saturate(None=DotProduct(A=GetMaterialAttributes(None=@n17.Result) {attributeGetTypes [#1]}.Normal, B=@n6))
  n19 = LinearInterpolate(
   A=VectorParameter() {parameterName=Flame Color, defaultValue=<rgba 0 12 9.3519 1>}.RGB
   B=VectorParameter() {parameterName=Flame Color B, defaultValue=<rgba 0.2678 0.7833 1.05 1>}.RGB
   Alpha=@n18
  )
  n20 = Multiply(A=TextureCoordinate(), B=ScalarParameter() {defaultValue=3, parameterName=Flame Tiling})
  n21 = BumpOffset(Coordinate=@n20, Height=ScalarParameter() {defaultValue=-2, parameterName=Flame Bump Height})
  n22 = AppendVector(A=ScalarParameter() {defaultValue=0.05, parameterName=Flame Panner X}, B=ScalarParameter() {defaultValue=0.2, parameterName=Flame Panner Y})
  n23 = Panner(
   Coordinate=Multiply(A=@n20, B=ScalarParameter() {defaultValue=1.25, parameterName=WarpFrequency}) {constB=6.8}
   Speed=AppendVector(
    A=ScalarParameter() {defaultValue=0.33, parameterName=Flame Panner X2}
    B=ScalarParameter() {defaultValue=-0.33, parameterName=Flame Panner Y2}
   )
  )
  n24 = TextureSample(UVs=@n23) {texture=&T_NoisePack_01}
  n25 = AppendVector(A=@n24.R, B=@n24.R)
  n26 = ScalarParameter() {defaultValue=0.02, parameterName=Warp Intensity}
  n27 = TextureSample(UVs=Add(A=Panner(Coordinate=@n21, Speed=@n22), B=Multiply(A=@n25, B=@n26))) {texture=/Game/Weapons/WeaponSkins/Textures/T_Swirly_Noise_D1}
  n28 = Add(A=OneMinus(None=@n23), B=Multiply(A=@n18, B=ScalarParameter() {defaultValue=0.5, parameterName=Edge Warp}))
  n29 = Saturate(None=Multiply(A=@n27.R, B=TextureSample(UVs=@n28) {texture=&T_NoisePack_01}.B))
  n30 = Saturate(None=Power(Base=Saturate(None=@n17.Fresnel), Exp=ScalarParameter() {defaultValue=7, parameterName=GlassFresnel Power}) {constExponent=8})
  n31 = LinearInterpolate(A=LinearInterpolate(B=@n19, Alpha=@n29) {constA=1}, B=Constant() {r=1}, Alpha=Power(Base=@n30)) {constB=0}
  Add(A=@n16, B=@n31)
  Constant()
  Constant() {r=5}
  Constant()
  n32 = SetMaterialAttributes!(
   MaterialAttributes=@n17.Result
   Base Color=Constant()
   Metallic=Constant()
   Specular=LinearInterpolate(A=ScalarParameter() {defaultValue=0.2, parameterName=Core Spec}, B=ScalarParameter() {defaultValue=0.9, parameterName=EdgeSpec}, Alpha=@n30)
   Roughness=ScalarParameter() {defaultValue=0.15, parameterName=Roughness}
   Emissive Color=Multiply(A=@n16, B=@n31)
  ) {attributeSetTypes [
    #2
    #3
    #4
    #5
    #6
   ]}
  MaterialFunctionCall(
   AO=@n17.AO
   Distance Blend=@n17.AO
   Fresnel=@n17.AO
   Material=BlendMaterialAttributes(
    A=@n17.Result
    B=@n32
    Alpha=TextureSample() {texture=/BRCosmetics/Accessories/FORT_Backpacks/Backpack_M_MED_Flame_Skull/Textures/T_M_MED_Flame_Skull_Backpack_FX}.R
   )
  ) {materialFunction=&BaseFN_CharacterFunction_PostFX}
  MaterialFunctionCall(In=@n27.R, Contrast=ScalarParameter() {defaultValue=0.1, parameterName=Flame Contrast}) {materialFunction=&CheapContrast}
  Multiply(A=@n29, B=@n19)
  Saturate(None=Add(A=@n10, B=@n30))
 ]
}
