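# Physics asset for the Flame Skull backpack cosmetic, providing collision bodies and joints for its dangling straps and zipper.
# Box bodies cover the left buckle and a three-link right buckle chain, capsules form a two-link zipper pull, and two capsules shape the pack itself.
# Each chain hangs from the pack body through angular-limited constraints, with the left buckle and right chain tip restricted to a single swing axis.
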PhysicsAsset /BRCosmetics/Accessories/FORT_Backpacks/Backpack_M_MED_Flame_Skull/Meshes/M_MED_Flame_Skull_Phy
bodies {
 dyn_buckle_l [{shapeType=Box, extentX=4, extentY=5.5, extentZ=1.8, center=<1.5 0 0>}]
 dyn_buckle_r_01 [{shapeType=Box, extentX=2, extentY=3, extentZ=1.8, center=<1 0 0>}]
 dyn_buckle_r_02 [{shapeType=Box, extentX=2, extentY=3, extentZ=1.8, center=<1 0 0>}]
 dyn_buckle_r_03 [{shapeType=Box, extentX=4, extentY=5.5, extentZ=1.8, center=<1.5 0 0>}]
 dyn_zipper_01 [{shapeType=Capsule, length=0.5696, radius=0.8, center=<2.2744 -0.1999 0>, rotation=<rot 90 0 0>}]
 dyn_zipper_02 [{shapeType=Capsule, length=1, radius=0.8, center=<2 -0.1999 0>, rotation=<rot 90 0 0>}]
 pack [
  {shapeType=Capsule, length=1, radius=1}
  {shapeType=Capsule, length=17.8737, radius=4.9111, center=<2.1776 -9.8871 -6.909>, rotation=<rot -59.9994 0 0>}
 ]
}
constraints [
 {bone1Name=dyn_buckle_r_01, bone2Name=pack, swing1LimitDegrees=20, swing2LimitDegrees=30, twistLimitDegrees=5}
 {bone1Name=dyn_buckle_r_02, bone2Name=dyn_buckle_r_01, swing1LimitDegrees=20, swing2LimitDegrees=30, twistLimitDegrees=5}
 {
  bone1Name=dyn_buckle_r_03
  bone2Name=dyn_buckle_r_02
  swing1LimitDegrees=45
  swing1Motion=Locked
  swing2LimitDegrees=30
  twistLimitDegrees=45
  twistMotion=Locked
 }
 {
  bone1Name=dyn_buckle_l
  bone2Name=pack
  swing1LimitDegrees=45
  swing1Motion=Locked
  swing2LimitDegrees=60
  twistLimitDegrees=45
  twistMotion=Locked
 }
 {bone1Name=dyn_zipper_01, bone2Name=pack, swing1LimitDegrees=30, swing2LimitDegrees=45, twistLimitDegrees=5}
 {bone1Name=dyn_zipper_02, bone2Name=dyn_zipper_01, swing1LimitDegrees=15, swing2LimitDegrees=45, twistLimitDegrees=2}
]
